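# Animation Blueprint for the female FearCatch head cosmetic that raises the material's temporal responsiveness when its flipbook face changes.
# Each update calls the parent, then reads eye and mouth flipbook index curves from the part mesh's anim instance and compares them to cached values.
# Any change sets the "Temporal Responsiveness" scalar on the mesh's materials to 1, otherwise 0.5, then caches the new indices.
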AnimBlueprint /BRCosmetics/Characters/Player/Female/Medium/Heads/F_MED_FearCatch_Head/Meshes/F_MED_FearCatch_Head_AnimBP
target_skeleton=/Game/Characters/Player/Common/Fortnite_Base_Head/Fortnite_Base_Head_Export_Skeleton
graphs [EventGraph {nodes [
   n1 = Event BlueprintUpdateAnimation()
   > Parent: BlueprintUpdateAnimation!(execute, DeltaTimeX:float=@n1.DeltaTimeX) -> (then) {is=title K2Node_CallParentFunction}
   n2 = GetPartSkeletalMeshComponent()
   n3 = GetAnimInstance(self=@n2)
   n4 = GetCurveValue(self=@n3, CurveName=flipbook_face_R_eye_index)
   n5 = "Not Equal (Float)"(A=@n4, B=Get Right Eye Index() -> (Right Eye Index:float))
   n6 = GetCurveValue(self=@n3, CurveName=flipbook_face_L_eye_index)
   n7 = "Not Equal (Float)"(A=@n6, B=Get Left Eye Index() -> (Left Eye Index:float))
   n8 = GetCurveValue(self=@n3, CurveName=flipbook_face_mouth_index)
   n9 = "Not Equal (Float)"(A=@n8, B=Get Mouth Index() -> (Mouth Index:float))
   > SetScalarParameterValueOnMaterials!(
    execute
    self:Mesh Component Object Reference=@n2
    ParameterName:Name=Temporal Responsiveness
    ParameterValue:float=SelectFloat(A=1, B=0.5, bPickA=OR Boolean!(A=@n5, B=@n7, C:bool=@n9) -> (ReturnValue))
   ) -> (then)
   > Set Left Eye Index(Left Eye Index:float=@n6) -> (Output_Get:float)
   > Set Right Eye Index(Right Eye Index:float=@n4) -> (Output_Get:float)
   > Set Mouth Index(Mouth Index:float=@n8) -> (Output_Get:float)
   TryGetPawnOwner()
  ]}]
variables [
 Left Brow Index {type=number}
 Left Eye Index {type=number}
 Mouth Index {type=number}
 Right Eye Index {type=number}
]
